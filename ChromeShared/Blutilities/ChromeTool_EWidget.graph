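# Editor utility widget for tagging actors with Gameplay.Status.Chrome and previewing their chrome material in the editor.
# Apply adds the tag to selected actors and calls ApplyChromeMaterial on each; Remove only strips the tag, leaving materials untouched.
# Preview/Hide runs the same ApplyChromeMaterial call on every tagged actor, with no toggle state in the widget itself.
# Both material paths pass scratch chrome and original material struct arrays, clearing them after each actor.
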EditorUtilityWidgetBlueprint /ChromeShared/Blutilities/ChromeTool_EWidget
graphs [EventGraph {nodes [
   "On Clicked (btnApplyChrome)"!() -> (OutputDelegate, then, widget:User Widget Object Reference, event:Pointer Event Structure) {is=title K2Node_ComponentBoundEvent}
   > n1 = GetSelectedActors()
   > n2 = For Each Loop(Array:Array of Actor Object References=@n1) -> (Array Element:Actor Object Reference)
   > n3 = Set ExistingActorTags(ExistingActorTags:Array of Names=Get Tags!(self:Actor Object Reference=@n2) -> (Tags:Array of Names)) -> (Output_Get:Array of Names)
   > Add Unique(TargetArray:Array of Names=@n3, NewItem:"Name (by ref)"=Get ChromeVisualActorTag() -> (ChromeVisualActorTag:Name=None))
   > Set Tags!(execute, Tags:Array of Names=Get ExistingActorTags() -> (ExistingActorTags:Array of Names), self:Actor Object Reference=@n2) -> (then, Output_Get:Array of Names)
   n4 = Get Empty Chrome Material Struct In() -> (Empty Chrome Material Struct In:Array of B Lackluster Chrome Material Struct Structures)
   n5 = Get Empty Original Material Struct In() -> (Empty Original Material Struct In:Array of B Lackluster Original Material Struct Structures)
   > ApplyChromeMaterial!(
    execute
    Actor:Actor Object Reference=@n2
    ChromeMaterialStructIn:Array of B Lackluster Chrome Material Struct Structures=@n4
    OriginalMaterialStructIn:Array of B Lackluster Original Material Struct Structures=@n5
    ChromeType:B_Lackluster_ChromeTypes Enum=NewEnumerator1
    OverrideActorType:B_Lackluster_ActorTypes Enum=NewEnumerator8
    __WorldContext:Object Reference
   ) -> (then)
   > Clear(TargetArray:Array of B Lackluster Chrome Material Struct Structures=@n4)
   > Clear(TargetArray:Array of B Lackluster Original Material Struct Structures=@n5)
   "On Clicked (btnToggleChromePreview)"!() -> (OutputDelegate, then, widget:User Widget Object Reference, event:Pointer Event Structure) {is=title K2Node_ComponentBoundEvent}
   > n6 = GetAllActorsWithTag!(execute, WorldContextObject:Object Reference, Tag:Name=Get ChromeVisualActorTag() -> (ChromeVisualActorTag=None)) -> (then, OutActors:Array of Actor Object References)
   > n7 = For Each Loop(Array=@n6)
   n8 = Get Empty Original Material Struct In()
   n9 = Get Empty Chrome Material Struct In()
   > ApplyChromeMaterial(Actor=@n7, ChromeMaterialStructIn=@n9, OriginalMaterialStructIn=@n8, ChromeType=NewEnumerator1, OverrideActorType=NewEnumerator8)
   > Clear(TargetArray:Array of B Lackluster Chrome Material Struct Structures=@n9)
   > Clear(TargetArray:Array of B Lackluster Original Material Struct Structures=@n8)
   "On Clicked (btnRemoveChrome)"!() -> (OutputDelegate, then, widget:User Widget Object Reference, event:Pointer Event Structure) {is=title K2Node_ComponentBoundEvent}
   > n10 = GetSelectedActors()
   > n11 = For Each Loop(Array=@n10)
   > n12 = Set ExistingActorTags(ExistingActorTags=Get Tags(self=@n11))
   > n13 = Remove Item!(execute, TargetArray:Array of Names=@n12, Item:"Name (by ref)"=Get ChromeVisualActorTag() -> (ChromeVisualActorTag=None)) -> (then, ReturnValue:bool)
   > Branch(Condition=@n13)
   > Set Tags(Tags=Get ExistingActorTags(), self=@n11)
  ]}]
variables [
 ChromeVisualActorTag {type=string, default=Gameplay.Status.Chrome}
 Empty Chrome Material Struct In {type="array<B_Lackluster_ChromeMaterialStruct>", default []}
 Empty Original Material Struct In {type="array<B_Lackluster_OriginalMaterialStruct>", default []}
 ExistingActorTags {type="array<string>", default []}
]
widgets {
 info {
  parentClass=/Game/Blueprints/Blutilities/ForTechArt/Widgets/TA_Base_EditorWidget.TA_Base_EditorWidget_C
  rootWidgetClass=&VerticalBox
  widgetCount=4
 }
 widgets [
  VerticalBox_1 {widgetClassPath=&VerticalBox, widget=VerticalBox_1}
  btnApplyChrome {
   widgetClassPath=&TA_Generic01_Button
   parent=VerticalBox_1
   widget=btnApplyChrome
   properties {label="Apply Chrome Tag to Selected Object(s)", toolTipText=Applies Gameplay.Status.Chrome tag to all selected Actors}
   slot {class=VerticalBoxSlot, properties {padding=<ltrb 0 5 0 0>}}
  }
  btnRemoveChrome {
   widgetClassPath=&TA_Generic01_Button
   parent=VerticalBox_1
   widget=btnRemoveChrome
   properties {label="Remove Chrome Tag from Selected Object(s)", toolTipText=Removes Gameplay.Status.Chrome from all selected Actors}
   slot {class=VerticalBoxSlot, properties {padding=<ltrb 0 5 0 0>}}
  }
  btnToggleChromePreview {
   widgetClassPath=&TA_Generic01_Button
   parent=VerticalBox_1
   widget=btnToggleChromePreview
   properties {
    label=Preview/Hide Chrome on All Objects with Tags
    toolTipText=Shows/Hides the Chrome Material on all actors with Gameplay.Status.Chrome tag
   }
   slot {class=VerticalBoxSlot, properties {padding=<ltrb 0 5 0 0>}}
  }
 ]
}
